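# UI sound cue, suffixed "old", for releasing a small holdable front-end button, routed to the RenderedCinematics sound class.
# A single quiet press-and-hold release wave passes through a Modulator node left at default settings.
# The cue plays at reduced volume and overrides concurrency with a short retrigger time.
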SoundCue /Game/Sounds/Fort_UI_Sounds/FrontEnd/New_FrontEnd/DefaultSounds/HoldableButtons/Small_Quiet_Clear/RenderedCine/UI_RenderedCine_H_oldable_Small_Release_Cue_old
root=Modulator_1
nodes [
 {id=Modulator_1, children [WavePlayer_1]}
 {
  id=WavePlayer_1
  properties {soundWaveAssetPtr=/Game/Sounds/Fort_UI_Sounds/FrontEnd/DefaultSounds/HoldableButtons/wav/UI_Button_Press_Hold_Quiet_Release}
 }
]
properties {
 bOverrideConcurrency=true
 maxDistance=2097152
 soundClassObject=/Game/Sounds/FortSoundClassesAndModes/RenderedCinematics
 volumeMultiplier=0.252
 concurrencyOverrides {retriggerTime=0.1}
}
